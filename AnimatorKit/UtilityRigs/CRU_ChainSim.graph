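# Utility Control Rig that procedurally builds a bone chain with simulated secondary motion.
# Construction spawns five bones along a straight B-spline, parents them in sequence under chainSim_main, and adds an attachToMe control to each.
# It also spawns a Global control carrying dynamics animation channels; an Alpha channel node is left unexecuted.
# Forward solve constrains chainSim_main to Global, then, when Enabled, feeds the channel values and bones into Local Chain Dynamics.
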ControlRigBlueprint /AnimatorKit/UtilityRigs/CRU_ChainSim
graphs [
 Concat_2_ContainedGraph {nodes [
   n1 = RigVMFunctionEntryNode()
   RigVMFunctionReturnNode(Result=/Script/RigVM.RigVMFunction_StringConcat(A=&RigVMFunction_StringConcat(A=@n1.A, B=@n1.B) {template="Concat::Execute(in A,in B,out Result)"}.Result, B=@n1.C) {is=struct RigVMUnitNode, template="Concat::Execute(in A,in B,out Result)"}.Result)
  ]}
 RigVMFunctionLibrary {links [], nodes []}
 RigVMFunction_Sequence_1_ContainedGraph {nodes [
   n1 = RigVMFunctionEntryNode()
   RigVMFunction_Sequence = &RigVMFunction_Sequence(ExecuteContext=@n1.ExecuteContext)
   n2 = &RigVMFunction_Sequence(ExecuteContext=@RigVMFunction_Sequence.B)
   RigVMFunctionReturnNode(A=@RigVMFunction_Sequence.A, B=@n2.A, C=@n2.B)
  ]}
 RigVMModel {nodes [
   n1 = &RigUnit_PrepareForExecution()
   &RigUnit_HierarchyImportFromSkeleton(NameSpace=None, bIncludeCurves=False, bIncludeMeshSockets=False, bIncludeVirtualBones=false, ExecutePin=@n1.ExecutePin)
   > RigVMFunction_Sequence = RigVMAggregateNode()
   RigVMVariableNode(Value=5, ExecuteContext=@RigVMFunction_Sequence.A) {variable=Number of Bones}
   > n2 = /Script/ControlRig.RigUnit_HierarchyAddBone(
    Name=chainSim_main
    Parent=(Type=Bone,Name="root")
    Space=GlobalSpace
    Transform=(Rotation=(X=0,Y=0,Z=0,W=1),Translation=(X=0,Y=0,Z=0),Scale3D=(X=1,Y=1,Z=1))
   ) {is=struct RigVMUnitNode}
   > RigVMVariableNode(Value=()) {variable=Bones}
   > RigVMFunction_ForLoopCount = /Script/RigVM.RigVMFunction_ForLoopCount(Count=RigVMVariableNode() {variable=Number of Bones}.Value) {is=struct RigVMUnitNode}
   To_String = RigVMDispatchNode(Value=/Script/RigVM.RigVMFunction_MathIntAdd(B=1, A=@RigVMFunction_ForLoopCount.Index) {is=struct RigVMUnitNode, template="Add::Execute(in A,in B,out Result)"}.Result) {template="DISPATCH_RigDispatch_ToString(in Value,out Result)"}
   Concat = &RigVMFunction_StringConcat(A=chainSim_, B=@To_String.Result) {template="Concat::Execute(in A,in B,out Result)"}
   From_String = RigVMDispatchNode(String=@Concat.Result) {template="DISPATCH_RigDispatch_FromString(in String,out Result)"}
   n3 = /Script/ControlRigSpline.RigUnit_ControlRigSplineFromPoints(
    Compression=0
    Points=((X=0,Y=0,Z=0),(X=0,Y=50,Z=0),(X=0,Y=100,Z=0),(X=0,Y=150,Z=0))
    SamplesPerSegment=16
    SplineMode=BSpline
    Stretch=0
    bClosed=False
   ) {is=struct RigVMUnitNode}
   > n4 = &RigUnit_HierarchyAddBone(
    Space=GlobalSpace
    Name=@From_String.Result
    Transform=/Script/ControlRigSpline.RigUnit_TransformFromControlRigSpline2(PrimaryAxis=(X=1,Y=0,Z=0), SecondaryAxis=(X=0,Y=0,Z=0), Spline=@n3.Spline, U=@RigVMFunction_ForLoopCount.Ratio) {is=struct RigVMUnitNode}.Transform
    Parent=@n2.Item
   )
   > Add = RigVMDispatchNode(Element=@n4.Item, Array=RigVMVariableNode() {variable=Bones}.Value) {template="DISPATCH_RigVMDispatch_ArrayAdd(io Array,in Element,out Index)"}
   Concat_1 = RigVMAggregateNode(A=attachToMe_, B=@To_String.Result)
   From_String_1 = RigVMDispatchNode(String=@Concat_1.Result) {template="DISPATCH_RigDispatch_FromString(in String,out Result)"}
   > n5 = /Script/ControlRig.RigUnit_HierarchyAddControlTransform(
    InitialValue=(Rotation=(X=0,Y=0,Z=0,W=1),Translation=(X=0,Y=0,Z=0),Scale3D=(X=1,Y=1,Z=1))
    OffsetSpace=LocalSpace
    OffsetTransform=(Rotation=(X=0,Y=0,Z=0,W=1),Translation=(X=0,Y=0,Z=0),Scale3D=(X=1,Y=1,Z=1))
    Settings="(InitialSpace=LocalSpace,bUsePreferredRotationOrder=False,PreferredRotationOrder=YZX,Limits=(LimitTranslationX=(bMinimum=False,bMaximum=False),LimitTranslationY=(bMinimum=False,bMaximum=False),LimitTranslationZ=(bMinimum=False,bMaximum=False),LimitPitch=(b ...(truncated)"
    Name=@From_String_1.Result
    Parent=@n4.Item
   ) {
    is=struct RigVMUnitNode
    template="SpawnControl::Execute(in InitialValue,in Settings,in OffsetTransform,in OffsetSpace,in Parent,in Name,out Item)"
   }
   Not_Equals = RigVMDispatchNode(B=0, A=@RigVMFunction_ForLoopCount.Index) {template="DISPATCH_RigVMDispatch_CoreNotEquals(in A,in B,out Result)"}
   RigVMFunction_ControlFlowBranch = &RigVMFunction_ControlFlowBranch(Condition=@Not_Equals.Result, ExecuteContext=@n5.ExecutePin)
   Subtract = /Script/RigVM.RigVMFunction_MathIntSub(B=1, A=@RigVMFunction_ForLoopCount.Index) {is=struct RigVMUnitNode, template="Subtract::Execute(in A,in B,out Result)"}
   /Script/ControlRig.RigUnit_SetDefaultParent(
    Parent=RigVMDispatchNode(Index=@Subtract.Result, Array=RigVMVariableNode() {variable=Bones}.Value) {template="DISPATCH_RigVMDispatch_ArrayGetAtIndex(in Array,in Index,out Element)"}.Element
    Child=@n4.Item
    ExecutePin=@RigVMFunction_ControlFlowBranch.True
   ) {is=struct RigVMUnitNode}
   SpawnControl = &RigUnit_HierarchyAddControlTransform(
    InitialValue=(Rotation=(X=0,Y=0,Z=0,W=1),Translation=(X=0,Y=0,Z=0),Scale3D=(X=1,Y=1,Z=1))
    Name=Global
    OffsetSpace=LocalSpace
    OffsetTransform=(Rotation=(X=0,Y=0,Z=0,W=1),Translation=(X=0,Y=0,Z=0),Scale3D=(X=1,Y=1,Z=1))
    Parent=(Type=Null,Name="ComponentSpace")
    Settings="(InitialSpace=LocalSpace,bUsePreferredRotationOrder=False,PreferredRotationOrder=YZX,Limits=(LimitTranslationX=(bMinimum=False,bMaximum=False),LimitTranslationY=(bMinimum=False,bMaximum=False),LimitTranslationZ=(bMinimum=False,bMaximum=False),LimitPitch=(b ...(truncated)"
    ExecutePin=@RigVMFunction_Sequence.B
   ) {template="SpawnControl::Execute(in InitialValue,in Settings,in OffsetTransform,in OffsetSpace,in Parent,in Name,out Item)"}
   > RigVMVariableNode(Value=@SpawnControl.Item) {variable=Control}
   > SpawnAnimationChannel = /Script/ControlRig.RigUnit_HierarchyAddAnimationChannelFloat(
    InitialValue=4
    LimitsEnabled=(Enabled=(bMinimum=True,bMaximum=True))
    MaximumValue=100
    MinimumValue=0.001
    Name=Strength
    Parent=@SpawnControl.Item
   ) {
    is=struct RigVMUnitNode
    template="SpawnAnimationChannel::Execute(in InitialValue,in MinimumValue,in MaximumValue,in LimitsEnabled,in Parent,in Name,out Item)"
   }
   > &RigUnit_HierarchyAddAnimationChannelFloat(
    InitialValue=0.5
    LimitsEnabled=(Enabled=(bMinimum=True,bMaximum=True))
    MaximumValue=100
    MinimumValue=0.001
    Name=Damping
    Parent=@SpawnControl.Item
   ) {template="SpawnAnimationChannel::Execute(in InitialValue,in MinimumValue,in MaximumValue,in LimitsEnabled,in Parent,in Name,out Item)"}
   > &RigUnit_HierarchyAddAnimationChannelFloat(
    InitialValue=1
    LimitsEnabled=(Enabled=(bMinimum=True,bMaximum=True))
    MaximumValue=1
    MinimumValue=0
    Name=Blend Start
    Parent=@SpawnControl.Item
   ) {template="SpawnAnimationChannel::Execute(in InitialValue,in MinimumValue,in MaximumValue,in LimitsEnabled,in Parent,in Name,out Item)"}
   > &RigUnit_HierarchyAddAnimationChannelFloat(
    InitialValue=1
    LimitsEnabled=(Enabled=(bMinimum=True,bMaximum=True))
    MaximumValue=1
    MinimumValue=0
    Name=Blend End
    Parent=@SpawnControl.Item
   ) {template="SpawnAnimationChannel::Execute(in InitialValue,in MinimumValue,in MaximumValue,in LimitsEnabled,in Parent,in Name,out Item)"}
   > &RigUnit_HierarchyAddAnimationChannelFloat(
    InitialValue=1
    LimitsEnabled=(Enabled=(bMinimum=True,bMaximum=True))
    MaximumValue=1
    MinimumValue=0
    Name=Blend Twist
    Parent=@SpawnControl.Item
   ) {template="SpawnAnimationChannel::Execute(in InitialValue,in MinimumValue,in MaximumValue,in LimitsEnabled,in Parent,in Name,out Item)"}
   > /Script/ControlRig.RigUnit_HierarchyAddAnimationChannelBool(InitialValue=true, LimitsEnabled=(), MaximumValue=True, MinimumValue=False, Name=Enabled, Parent=@SpawnControl.Item) {
    is=struct RigVMUnitNode
    template="SpawnAnimationChannel::Execute(in InitialValue,in MinimumValue,in MaximumValue,in LimitsEnabled,in Parent,in Name,out Item)"
   }
   RigUnit_BeginExecution = &RigUnit_BeginExecution()
   n6 = &RigUnit_ParentConstraint(
    AdvancedSettings=(InterpolationType=Average,RotationOrderForFilter=XZY)
    Child=(Type=Bone,Name="chainSim_main")
    Parents=((Item=(Type=Bone,Name="None"),Weight=1))
    Weight=1
    ExecutePin=@RigUnit_BeginExecution.ExecutePin
    Parents.0.Item=RigVMVariableNode() {variable=Control}.Value
   )
   n7 = &RigVMFunction_ControlFlowBranch(
    Condition=/Script/ControlRig.RigUnit_GetBoolAnimationChannelFromItem(Item=(Type=Control,Name="Enabled"), bInitial=False) {is=struct RigVMUnitNode, template="GetAnimationChannelFromItem::Execute(out Value,in Item,in bInitial)"}.Value
    ExecuteContext=@n6.ExecutePin
   )
   GetAnimationChannelFromItem = /Script/ControlRig.RigUnit_GetFloatAnimationChannelFromItem(Item=(Type=Control,Name="Strength"), bInitial=False) {is=struct RigVMUnitNode, template="GetAnimationChannelFromItem::Execute(out Value,in Item,in bInitial)"}
   Local Chain Dynamics = RigVMFunctionReferenceNode(
    Draw Debug=false
    Strength=@GetAnimationChannelFromItem.Value
    Damping=&RigUnit_GetFloatAnimationChannelFromItem(Item=(Type=Control,Name="Damping"), bInitial=False) {template="GetAnimationChannelFromItem::Execute(out Value,in Item,in bInitial)"}.Value
    Blend Start=&RigUnit_GetFloatAnimationChannelFromItem(Item=(Type=Control,Name="Blend Start"), bInitial=False) {template="GetAnimationChannelFromItem::Execute(out Value,in Item,in bInitial)"}.Value
    Blend End=&RigUnit_GetFloatAnimationChannelFromItem(Item=(Type=Control,Name="Blend End"), bInitial=False) {template="GetAnimationChannelFromItem::Execute(out Value,in Item,in bInitial)"}.Value
    Blend Twist=&RigUnit_GetFloatAnimationChannelFromItem(Item=(Type=Control,Name="Blend Twist"), bInitial=False) {template="GetAnimationChannelFromItem::Execute(out Value,in Item,in bInitial)"}.Value
    ExecuteContext=@n7.True
    Bones=RigVMVariableNode() {variable=Bones}.Value
   )
   &RigUnit_HierarchyAddAnimationChannelFloat(
    ExecutePin=()
    InitialValue=1
    LimitsEnabled=(Enabled=(bMinimum=True,bMaximum=True))
    MaximumValue=1
    MinimumValue=0
    Name=Alpha
    Parent=@SpawnControl.Item
   ) {template="SpawnAnimationChannel::Execute(in InitialValue,in MinimumValue,in MaximumValue,in LimitsEnabled,in Parent,in Name,out Item)"}
  ]}
]
hierarchy [ComponentSpace {type=Null, loc=<0 0 0>, rot=<0 0 0 1>}]
variables [
 Bones {type="TArray<FRigElementKey>", default=(())}
 Control {type=FRigElementKey, default=(Type=Bone,Name="")}
 Number of Bones {type=int32, default=6}
]
