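# Utility Control Rig that exposes the morph targets named in the MorphTargets array as animatable channels.
# Construction imports the skeleton with curves, then spawns a MorphTargets float control with a 0-1 channel per name.
# Forward solve writes each existing curve as its channel value times the parent control, which acts as a master weight.
# A timeline-throttled check, over three seconds apart, prints missing curve names and a hint about recreating the rig.
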ControlRigBlueprint /AnimatorKit/UtilityRigs/CRU_MorphTargetDriver
graphs [
 RigVMFunctionLibrary {links [], nodes []}
 RigVMModel {nodes [
   n1 = &RigUnit_PrepareForExecution()
   RigVMVariableNode(Value=-10000, ExecuteContext=@n1.ExecutePin) {variable=TimeSinceLastError}
   > &RigUnit_HierarchyImportFromSkeleton(NameSpace=None, bIncludeCurves=true, bIncludeMeshSockets=False, bIncludeVirtualBones=false)
   > SpawnControl = /Script/ControlRig.RigUnit_HierarchyAddControlFloat(
    InitialValue=1
    Name=MorphTargets
    OffsetSpace=LocalSpace
    OffsetTransform=(Rotation=(X=0,Y=0,Z=0,W=1),Translation=(X=0,Y=0,Z=0),Scale3D=(X=1,Y=1,Z=1))
    Parent=(Type=Bone,Name="None")
    Settings="(PrimaryAxis=X,bIsScale=False,Limits=(Limit=(bMinimum=True,bMaximum=True),MinValue=0,MaxValue=1,bDrawLimits=False),Shape=(bVisible=False,Name=\"Default\",Color=(R=1,G=0,B=0,A=1),Transform=(Rotation=(X=0,Y=0,Z=0,W=1),Translation=(X=0,Y=0,Z=0),Scale3D=(X=1,Y=1 ...(truncated)"
   ) {
    is=struct RigVMUnitNode
    template="SpawnControl::Execute(in InitialValue,in Settings,in OffsetTransform,in OffsetSpace,in Parent,in Name,out Item)"
   }
   > For_Each = RigVMDispatchNode(Array=RigVMVariableNode() {variable=MorphTargets}.Value) {template="DISPATCH_RigVMDispatch_ArrayIterator(in Array,out Element,out Index,out Count,out Ratio)"}
   > SpawnScaleAnimationChannel = /Script/ControlRig.RigUnit_HierarchyAddAnimationChannelScaleFloat(
    InitialValue=1
    LimitsEnabled=(Enabled=(bMinimum=True,bMaximum=True))
    MaximumValue=1
    MinimumValue=0
    Parent=(Type=Control,Name="MorphTargets")
    Name=@For_Each.Element
   ) {
    is=struct RigVMUnitNode
    template="SpawnScaleAnimationChannel::Execute(in InitialValue,in MinimumValue,in MaximumValue,in LimitsEnabled,in Parent,in Name,out Item)"
   }
   RigUnit_BeginExecution = &RigUnit_BeginExecution()
   n2 = RigVMVariableNode() {variable=MorphTargets}
   For_Each_1 = RigVMDispatchNode(ExecuteContext=@RigUnit_BeginExecution.ExecutePin, Array=@n2.Value) {template="DISPATCH_RigVMDispatch_ArrayIterator(in Array,out Element,out Index,out Count,out Ratio)"}
   RigVMFunction_ControlFlowBranch = &RigVMFunction_ControlFlowBranch(
    Condition=/Script/ControlRig.RigUnit_CurveExists(Curve=@For_Each_1.Element) {is=struct RigVMUnitNode}.Exists
    ExecuteContext=@For_Each_1.ExecuteContext
   )
   RigUnit_GetControlFloat = /Script/ControlRig.RigUnit_GetControlFloat(Control=@For_Each_1.Element) {is=struct RigVMUnitNode, template="GetControlValue::Execute(in Control,out FloatValue,out Minimum,out Maximum)"}
   Multiply = &RigVMFunction_MathDoubleMul(
    A=@RigUnit_GetControlFloat.FloatValue
    B=&RigUnit_GetControlFloat(Control=MorphTargets) {template="GetControlValue::Execute(in Control,out FloatValue,out Minimum,out Maximum)"}.FloatValue
   )
   /Script/ControlRig.RigUnit_SetCurveValue(Curve=@For_Each_1.Element, Value=@Multiply.Result, ExecutePin=@RigVMFunction_ControlFlowBranch.True) {is=struct RigVMUnitNode}
   RigVMFunction_Timeline = /Script/RigVM.RigVMFunction_Timeline(Speed=1) {is=struct RigVMUnitNode}
   Subtract = &RigVMFunction_MathDoubleSub(A=@RigVMFunction_Timeline.Time, B=RigVMVariableNode() {variable=TimeSinceLastError}.Value)
   Greater = /Script/RigVM.RigVMFunction_MathDoubleGreater(B=3, A=@Subtract.Result) {is=struct RigVMUnitNode, template="Greater::Execute(in A,in B,out Result)"}
   n3 = &RigVMFunction_ControlFlowBranch(ExecuteContext=@For_Each_1.Completed, Condition=@Greater.Result)
   RigVMVariableNode(ExecuteContext=@n3.True) {variable=HasMissingMorphTargets}
   > For_Each_2 = RigVMDispatchNode(Array=@n2.Value) {template="DISPATCH_RigVMDispatch_ArrayIterator(in Array,out Element,out Index,out Count,out Ratio)"}
   n4 = &RigVMFunction_ControlFlowBranch(Condition=&RigUnit_CurveExists(Curve=@For_Each_2.Element).Exists, ExecuteContext=@For_Each_2.ExecuteContext)
   Concat = /Script/RigVM.RigVMFunction_NameConcat(B=" not found!", A=@For_Each_2.Element) {is=struct RigVMUnitNode, template="Concat::Execute(in A,in B,out Result)"}
   To_String = RigVMDispatchNode(Value=@Concat.Result) {template="DISPATCH_RigDispatch_ToString(in Value,out Result)"}
   RigVMDispatchNode(
    Enabled=True
    Prefix="CRU_MorphTargetDriver: "
    ScreenColor=(R=1,G=0,B=0,A=1)
    ScreenDuration=3
    ExecuteContext=@n4.False
    Value=@To_String.Result
   ) {template="DISPATCH_RigVMDispatch_Print(in Prefix,in Value,in Enabled,in ScreenDuration,in ScreenColor)"}
   > RigVMVariableNode(Value=true) {variable=HasMissingMorphTargets}
   n5 = RigVMVariableNode(ExecuteContext=@For_Each_2.Completed, Value=@RigVMFunction_Timeline.Time) {variable=TimeSinceLastError}
   n6 = &RigVMFunction_ControlFlowBranch(ExecuteContext=@n5.ExecuteContext, Condition=RigVMVariableNode() {variable=HasMissingMorphTargets}.Value)
   RigVMDispatchNode(
    Enabled=True
    Prefix="CRU_MorphTargetDriver: "
    ScreenColor=(R=0,G=0.783,B=0.1208,A=1)
    ScreenDuration=3
    Value="To set Morph Targets, right click the Control Rig Track -> Recreate Control Rig with New Settings"
    ExecuteContext=@n6.True
   ) {template="DISPATCH_RigVMDispatch_Print(in Prefix,in Value,in Enabled,in ScreenDuration,in ScreenColor)"}
  ]}
]
variables [
 HasMissingMorphTargets {type=bool, default=False}
 MorphTargets {type="TArray<FName>", default=("MorphTargetName","MorphTargetName2")}
 TimeSinceLastError {type=double, default=-1000000}
]
