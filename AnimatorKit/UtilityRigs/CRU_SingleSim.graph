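# CRU_SingleSim is an AnimatorKit utility Control Rig that adds springy secondary motion to a single transform.
# Construction spawns a Global control with Strength, Damping and Enabled channels, a chainSim_Main/chainSim bone pair under root, and an attachToMe control on chainSim.
# Forward solve constrains chainSim_Main to Global, then, when Enabled, spring-interpolates chainSim's global translation and rotation.
# The attachToMe control thus rides the simulated bone, lagging behind Global.
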ControlRigBlueprint /AnimatorKit/UtilityRigs/CRU_SingleSim
graphs [
 RigVMFunctionLibrary {links [], nodes []}
 RigVMFunction_Sequence_1_ContainedGraph {nodes [
   n1 = RigVMFunctionEntryNode()
   RigVMFunction_Sequence = &RigVMFunction_Sequence(ExecuteContext=@n1.ExecuteContext)
   n2 = &RigVMFunction_Sequence(ExecuteContext=@RigVMFunction_Sequence.B)
   RigVMFunctionReturnNode(A=@RigVMFunction_Sequence.A, B=@n2.A, C=@n2.B)
  ]}
 RigVMModel {nodes [
   n1 = &RigUnit_PrepareForExecution()
   &RigUnit_HierarchyImportFromSkeleton(NameSpace=None, bIncludeCurves=False, bIncludeMeshSockets=False, bIncludeVirtualBones=false, ExecutePin=@n1.ExecutePin)
   > RigVMFunction_Sequence = RigVMAggregateNode()
   SpawnControl = /Script/ControlRig.RigUnit_HierarchyAddControlTransform(
    InitialValue=(Rotation=(X=0,Y=0,Z=0,W=1),Translation=(X=0,Y=0,Z=0),Scale3D=(X=1,Y=1,Z=1))
    Name=Global
    OffsetSpace=LocalSpace
    OffsetTransform=(Rotation=(X=0,Y=0,Z=0,W=1),Translation=(X=0,Y=0,Z=0),Scale3D=(X=1,Y=1,Z=1))
    Parent=(Type=Null,Name="ComponentSpace")
    Settings="(InitialSpace=LocalSpace,bUsePreferredRotationOrder=False,PreferredRotationOrder=YZX,Limits=(LimitTranslationX=(bMinimum=False,bMaximum=False),LimitTranslationY=(bMinimum=False,bMaximum=False),LimitTranslationZ=(bMinimum=False,bMaximum=False),LimitPitch=(b ...(truncated)"
    ExecutePin=@RigVMFunction_Sequence.A
   ) {
    is=struct RigVMUnitNode
    template="SpawnControl::Execute(in InitialValue,in Settings,in OffsetTransform,in OffsetSpace,in Parent,in Name,out Item)"
   }
   > RigVMVariableNode(Value=@SpawnControl.Item) {variable=Control}
   > SpawnAnimationChannel = /Script/ControlRig.RigUnit_HierarchyAddAnimationChannelFloat(
    InitialValue=4
    LimitsEnabled=(Enabled=(bMinimum=True,bMaximum=True))
    MaximumValue=100
    MinimumValue=0.001
    Name=Strength
    Parent=@SpawnControl.Item
   ) {
    is=struct RigVMUnitNode
    template="SpawnAnimationChannel::Execute(in InitialValue,in MinimumValue,in MaximumValue,in LimitsEnabled,in Parent,in Name,out Item)"
   }
   > &RigUnit_HierarchyAddAnimationChannelFloat(
    InitialValue=0.5
    LimitsEnabled=(Enabled=(bMinimum=True,bMaximum=True))
    MaximumValue=100
    MinimumValue=0.001
    Name=Damping
    Parent=@SpawnControl.Item
   ) {template="SpawnAnimationChannel::Execute(in InitialValue,in MinimumValue,in MaximumValue,in LimitsEnabled,in Parent,in Name,out Item)"}
   > /Script/ControlRig.RigUnit_HierarchyAddAnimationChannelBool(InitialValue=true, LimitsEnabled=(), MaximumValue=True, MinimumValue=False, Name=Enabled, Parent=@SpawnControl.Item) {
    is=struct RigVMUnitNode
    template="SpawnAnimationChannel::Execute(in InitialValue,in MinimumValue,in MaximumValue,in LimitsEnabled,in Parent,in Name,out Item)"
   }
   n2 = /Script/ControlRig.RigUnit_HierarchyAddBone(
    Name=chainSim_Main
    Parent=(Type=Bone,Name="root")
    Space=GlobalSpace
    Transform=(Rotation=(X=0,Y=0,Z=0,W=1),Translation=(X=0,Y=0,Z=0),Scale3D=(X=1,Y=1,Z=1))
    ExecutePin=@RigVMFunction_Sequence.C
   ) {is=struct RigVMUnitNode}
   > n3 = &RigUnit_HierarchyAddBone(Name=chainSim, Space=GlobalSpace, Transform=(Rotation=(X=0,Y=0,Z=0,W=1),Translation=(X=0,Y=0,Z=0),Scale3D=(X=1,Y=1,Z=1)), Parent=@n2.Item)
   > RigVMVariableNode(Value=@n3.Item) {variable=Bone}
   > n4 = &RigUnit_HierarchyAddControlTransform(
    InitialValue=(Rotation=(X=0,Y=0,Z=0,W=1),Translation=(X=0,Y=0,Z=0),Scale3D=(X=1,Y=1,Z=1))
    Name=attachToMe
    OffsetSpace=LocalSpace
    OffsetTransform=(Rotation=(X=0,Y=0,Z=0,W=1),Translation=(X=0,Y=0,Z=0),Scale3D=(X=1,Y=1,Z=1))
    Settings="(InitialSpace=LocalSpace,bUsePreferredRotationOrder=False,PreferredRotationOrder=YZX,Limits=(LimitTranslationX=(bMinimum=False,bMaximum=False),LimitTranslationY=(bMinimum=False,bMaximum=False),LimitTranslationZ=(bMinimum=False,bMaximum=False),LimitPitch=(b ...(truncated)"
    Parent=@n3.Item
   ) {template="SpawnControl::Execute(in InitialValue,in Settings,in OffsetTransform,in OffsetSpace,in Parent,in Name,out Item)"}
   > /Script/ControlRig.RigUnit_SetDefaultParent(Child=@n4.Item, Parent=RigVMVariableNode() {variable=Bone}.Value) {is=struct RigVMUnitNode}
   RigUnit_BeginExecution = &RigUnit_BeginExecution()
   n5 = &RigUnit_ParentConstraint(
    AdvancedSettings=(InterpolationType=Average,RotationOrderForFilter=XZY)
    Child=(Type=Bone,Name="chainSim_Main")
    Parents=((Item=(Type=Bone,Name="None"),Weight=1))
    Weight=1
    ExecutePin=@RigUnit_BeginExecution.ExecutePin
    Parents.0.Item=RigVMVariableNode() {variable=Control}.Value
   )
   n6 = &RigVMFunction_ControlFlowBranch(
    Condition=/Script/ControlRig.RigUnit_GetBoolAnimationChannelFromItem(Item=(Type=Control,Name="Enabled"), bInitial=False) {is=struct RigVMUnitNode, template="GetAnimationChannelFromItem::Execute(out Value,in Item,in bInitial)"}.Value
    ExecuteContext=@n5.ExecutePin
   )
   GetAnimationChannelFromItem = /Script/ControlRig.RigUnit_GetFloatAnimationChannelFromItem(Item=(Type=Control,Name="Strength"), bInitial=False) {is=struct RigVMUnitNode, template="GetAnimationChannelFromItem::Execute(out Value,in Item,in bInitial)"}
   n7 = &RigUnit_GetFloatAnimationChannelFromItem(Item=(Type=Control,Name="Damping"), bInitial=False) {template="GetAnimationChannelFromItem::Execute(out Value,in Item,in bInitial)"}
   n8 = RigVMVariableNode() {variable=Bone}
   RigUnit_GetTransform = &RigUnit_GetTransform(bInitial=False, Item=@n8.Value)
   SpringInterp = /Script/ControlRig.RigUnit_SpringInterpVectorV2(
    Current=(X=0,Y=0,Z=0)
    Force=(X=0,Y=0,Z=0)
    TargetVelocityAmount=0
    bInitializeFromTarget=False
    bUseCurrentInput=False
    Strength=@GetAnimationChannelFromItem.Value
    CriticalDamping=@n7.Value
    Target=@RigUnit_GetTransform.Transform.Translation
   ) {
    is=struct RigVMUnitNode
    template="SpringInterp::Execute(in Target,in Strength,in CriticalDamping,in Force,in bUseCurrentInput,in Current,in TargetVelocityAmount,in bInitializeFromTarget,out Result,out Velocity)"
   }
   Set Transform = /Script/ControlRig.RigUnit_SetTranslation(Space=GlobalSpace, Weight=1, bInitial=False, bPropagateToChildren=True, ExecutePin=@n6.True, Value=@SpringInterp.Result, Item=@n8.Value) {is=struct RigVMUnitNode, template="Set Transform::Execute(in Item,in Space,in bInitial,in Value,in Weight,in bPropagateToChildren)"}
   n9 = /Script/ControlRig.RigUnit_SpringInterpQuaternionV2(
    Current=(X=0,Y=0,Z=0,W=1)
    TargetVelocityAmount=0
    Torque=(X=0,Y=0,Z=0)
    bInitializeFromTarget=False
    bUseCurrentInput=False
    Strength=@GetAnimationChannelFromItem.Value
    CriticalDamping=@n7.Value
    Target=@RigUnit_GetTransform.Transform.Rotation
   ) {
    is=struct RigVMUnitNode
    template="SpringInterp::Execute(in Target,in Strength,in CriticalDamping,in Torque,in bUseCurrentInput,in Current,in TargetVelocityAmount,in bInitializeFromTarget,out Result,out AngularVelocity)"
   }
   > Set Transform_1 = &RigUnit_SetRotation(Space=GlobalSpace, Weight=1, bInitial=False, Value=@n9.Result, Item=@n8.Value)
  ]}
]
hierarchy [ComponentSpace {type=Null, loc=<0 0 0>, rot=<0 0 0 1>}]
variables [
 Bone {type=FRigElementKey, default=(Type=Bone,Name="")}
 Control {type=FRigElementKey, default=(Type=Bone,Name="")}
]
