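# Control Rig utility that provides an editable spline path with an AttachToMe control that rides along it.
# Construction spawns a Global control, a splinePath_main bone, tagged controls spaced along a default B-spline, and the attach control with U, roll, auto-align and color channels.
# Forward solve constrains the bone to Global, rebuilds and draws the spline from the tagged controls, and positions the attach control's parent at U, optionally aligned.
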ControlRigBlueprint /AnimatorKit/UtilityRigs/CRU_SplinePath
graphs [
 Concat_2_ContainedGraph {nodes [
   n1 = RigVMFunctionEntryNode()
   RigVMFunctionReturnNode(Result=/Script/RigVM.RigVMFunction_StringConcat(A=&RigVMFunction_StringConcat(A=@n1.A, B=@n1.B) {template="Concat::Execute(in A,in B,out Result)"}.Result, B=@n1.C) {is=struct RigVMUnitNode, template="Concat::Execute(in A,in B,out Result)"}.Result)
  ]}
 RigVMFunctionLibrary {links [], nodes []}
 RigVMFunction_Sequence_1_ContainedGraph {nodes [
   n1 = RigVMFunctionEntryNode()
   RigVMFunction_Sequence = &RigVMFunction_Sequence(ExecuteContext=@n1.ExecuteContext)
   n2 = &RigVMFunction_Sequence(ExecuteContext=@RigVMFunction_Sequence.B)
   RigVMFunctionReturnNode(A=@RigVMFunction_Sequence.A, B=@n2.A, C=@n2.B)
  ]}
 RigVMModel {nodes [
   n1 = &RigUnit_PrepareForExecution()
   &RigUnit_HierarchyImportFromSkeleton(NameSpace=None, bIncludeCurves=False, bIncludeMeshSockets=False, bIncludeVirtualBones=false, ExecutePin=@n1.ExecutePin)
   > RigVMFunction_Sequence = RigVMAggregateNode()
   SpawnControl = /Script/ControlRig.RigUnit_HierarchyAddControlTransform(
    InitialValue=(Rotation=(X=0,Y=0,Z=0,W=1),Translation=(X=0,Y=0,Z=0),Scale3D=(X=1,Y=1,Z=1))
    Name=Global
    OffsetSpace=LocalSpace
    OffsetTransform=(Rotation=(X=0,Y=0,Z=0,W=1),Translation=(X=0,Y=0,Z=0),Scale3D=(X=1,Y=1,Z=1))
    Parent=(Type=Null,Name="ComponentSpace")
    Settings="(InitialSpace=LocalSpace,bUsePreferredRotationOrder=False,PreferredRotationOrder=YZX,Limits=(LimitTranslationX=(bMinimum=False,bMaximum=False),LimitTranslationY=(bMinimum=False,bMaximum=False),LimitTranslationZ=(bMinimum=False,bMaximum=False),LimitPitch=(b ...(truncated)"
    ExecutePin=@RigVMFunction_Sequence.A
   ) {
    is=struct RigVMUnitNode
    template="SpawnControl::Execute(in InitialValue,in Settings,in OffsetTransform,in OffsetSpace,in Parent,in Name,out Item)"
   }
   > RigVMVariableNode(Value=@SpawnControl.Item) {variable=Control}
   > Control Stack at Item = RigVMFunctionReferenceNode(
    Add Bottom Null=False
    Add Null=true
    Add Secondary Control=False
    Control Suffix=_ctrl
    Item=(Name="None",Type=None)
    Name=AttachToMe
    Null Suffix=_null
    Secondary Settings="(Shape=(bVisible=True,Name=\"Default\",Color=(R=1,G=0,B=0,A=1),Transform=(Rotation=(X=0,Y=0,Z=0,W=1),Translation=(X=0,Y=0,Z=0),Scale3D=(X=1,Y=1,Z=1))),Proxy=(bIsProxy=False,ShapeVisibility=BasedOnSelection),DisplayName=\"None\",InitialSpace=LocalSpace,bUsePref ...(truncated)"
    Settings="(Shape=(bVisible=True,Name=\"Diamond_Thick\",Color=(R=0,G=1,B=0,A=1),Transform=(Rotation=(X=0,Y=0,Z=0.3827,W=0.9239),Translation=(X=0,Y=0,Z=0),Scale3D=(X=2,Y=2,Z=2))),Proxy=(bIsProxy=False,ShapeVisibility=BasedOnSelection),DisplayName=\"None\",InitialSpace=Loc ...(truncated)"
    Tag Controls=true
    Zero Controls=true
    Parent=@SpawnControl.Item
   )
   > RigVMVariableNode(Value=@"Control Stack at Item".Control) {variable=Attach Control}
   > SpawnAnimationChannel = /Script/ControlRig.RigUnit_HierarchyAddAnimationChannelFloat(
    InitialValue=0
    LimitsEnabled=(Enabled=(bMinimum=True,bMaximum=True))
    MaximumValue=1
    MinimumValue=0
    Name=U Value
    Parent=@"Control Stack at Item".Control
   ) {
    is=struct RigVMUnitNode
    template="SpawnAnimationChannel::Execute(in InitialValue,in MinimumValue,in MaximumValue,in LimitsEnabled,in Parent,in Name,out Item)"
   }
   > &RigUnit_HierarchyAddAnimationChannelFloat(
    InitialValue=0
    LimitsEnabled=(Enabled=(bMinimum=false,bMaximum=false))
    MaximumValue=180
    MinimumValue=-180
    Name=Roll
    Parent=@"Control Stack at Item".Control
   ) {template="SpawnAnimationChannel::Execute(in InitialValue,in MinimumValue,in MaximumValue,in LimitsEnabled,in Parent,in Name,out Item)"}
   > /Script/ControlRig.RigUnit_HierarchyAddAnimationChannelBool(InitialValue=true, LimitsEnabled=(), MaximumValue=True, MinimumValue=False, Name=Auto Align, Parent=@"Control Stack at Item".Control) {
    is=struct RigVMUnitNode
    template="SpawnAnimationChannel::Execute(in InitialValue,in MinimumValue,in MaximumValue,in LimitsEnabled,in Parent,in Name,out Item)"
   }
   > &RigUnit_HierarchyAddAnimationChannelFloat(
    InitialValue=0
    LimitsEnabled=(Enabled=(bMinimum=True,bMaximum=True))
    MaximumValue=1
    MinimumValue=0
    Name=ColorR
    Parent=@"Control Stack at Item".Control
   ) {template="SpawnAnimationChannel::Execute(in InitialValue,in MinimumValue,in MaximumValue,in LimitsEnabled,in Parent,in Name,out Item)"}
   > &RigUnit_HierarchyAddAnimationChannelFloat(
    InitialValue=0
    LimitsEnabled=(Enabled=(bMinimum=True,bMaximum=True))
    MaximumValue=1
    MinimumValue=0
    Name=ColorG
    Parent=@"Control Stack at Item".Control
   ) {template="SpawnAnimationChannel::Execute(in InitialValue,in MinimumValue,in MaximumValue,in LimitsEnabled,in Parent,in Name,out Item)"}
   > &RigUnit_HierarchyAddAnimationChannelFloat(
    InitialValue=0
    LimitsEnabled=(Enabled=(bMinimum=True,bMaximum=True))
    MaximumValue=1
    MinimumValue=0
    Name=ColorB
    Parent=@"Control Stack at Item".Control
   ) {template="SpawnAnimationChannel::Execute(in InitialValue,in MinimumValue,in MaximumValue,in LimitsEnabled,in Parent,in Name,out Item)"}
   RigVMVariableNode(Value=5, ExecuteContext=@RigVMFunction_Sequence.B) {variable=Number of Bones}
   > /Script/ControlRig.RigUnit_HierarchyAddBone(
    Name=splinePath_main
    Parent=(Type=Bone,Name="root")
    Space=GlobalSpace
    Transform=(Rotation=(X=0,Y=0,Z=0,W=1),Translation=(X=0,Y=0,Z=0),Scale3D=(X=1,Y=1,Z=1))
   ) {is=struct RigVMUnitNode}
   > RigVMVariableNode(Value=()) {variable=Bones}
   > RigVMFunction_ForLoopCount = /Script/RigVM.RigVMFunction_ForLoopCount(Count=RigVMVariableNode() {variable=Number of Bones}.Value) {is=struct RigVMUnitNode}
   n2 = /Script/ControlRigSpline.RigUnit_ControlRigSplineFromPoints(
    Compression=0
    Points=((X=0,Y=0,Z=0),(X=0,Y=50,Z=0),(X=0,Y=100,Z=0),(X=0,Y=150,Z=0))
    SamplesPerSegment=16
    SplineMode=BSpline
    Stretch=0
    bClosed=False
   ) {is=struct RigVMUnitNode}
   > RigVMVariableNode(Value=@n2.Spline) {variable=Spline}
   To_String = RigVMDispatchNode(Value=/Script/RigVM.RigVMFunction_MathIntAdd(B=1, A=@RigVMFunction_ForLoopCount.Index) {is=struct RigVMUnitNode, template="Add::Execute(in A,in B,out Result)"}.Result) {template="DISPATCH_RigDispatch_ToString(in Value,out Result)"}
   Concat_1 = RigVMAggregateNode(A=spline_, C=_ctrl, B=@To_String.Result)
   From_String_1 = RigVMDispatchNode(String=@Concat_1.Result) {template="DISPATCH_RigDispatch_FromString(in String,out Result)"}
   > n3 = &RigUnit_HierarchyAddControlTransform(
    InitialValue=(Rotation=(X=0,Y=0,Z=0,W=1),Translation=(X=0,Y=0,Z=0),Scale3D=(X=1,Y=1,Z=1))
    OffsetSpace=LocalSpace
    Settings="(InitialSpace=LocalSpace,bUsePreferredRotationOrder=False,PreferredRotationOrder=YZX,Limits=(LimitTranslationX=(bMinimum=False,bMaximum=False),LimitTranslationY=(bMinimum=False,bMaximum=False),LimitTranslationZ=(bMinimum=False,bMaximum=False),LimitPitch=(b ...(truncated)"
    Name=@From_String_1.Result
    OffsetTransform=/Script/ControlRigSpline.RigUnit_TransformFromControlRigSpline2(PrimaryAxis=(X=1,Y=0,Z=0), SecondaryAxis=(X=0,Y=0,Z=0), Spline=@n2.Spline, U=@RigVMFunction_ForLoopCount.Ratio) {is=struct RigVMUnitNode}.Transform
    Parent=RigVMVariableNode() {variable=Control}.Value
   ) {template="SpawnControl::Execute(in InitialValue,in Settings,in OffsetTransform,in OffsetSpace,in Parent,in Name,out Item)"}
   > /Script/ControlRig.RigUnit_SetMetadataTag(NameSpace=Self, Tag=Spline Controls, Item=@n3.Item) {is=struct RigVMUnitNode}
   RigUnit_BeginExecution = &RigUnit_BeginExecution()
   &RigUnit_ParentConstraint(
    AdvancedSettings=(InterpolationType=Average,RotationOrderForFilter=XZY)
    Child=(Type=Bone,Name="splinePath_main")
    Parents=((Item=(Type=Bone,Name="None"),Weight=1))
    Weight=1
    ExecutePin=@RigUnit_BeginExecution.ExecutePin
    Parents.0.Item=RigVMVariableNode() {variable=Control}.Value
   )
   n4 = /Script/ControlRigSpline.RigUnit_ControlRigSplineFromTransforms(
    Compression=0
    SamplesPerSegment=16
    SplineMode=BSpline
    Stretch=0
    bClosed=false
    Transforms=/Script/ControlRig.RigUnit_GetTransformItemArray(
     Space=GlobalSpace
     bInitial=False
     Items=/Script/ControlRig.RigUnit_FindItemsWithMetadataTag(NameSpace=Self, Tag=Spline Controls) {is=struct RigVMUnitNode}.Items
    ) {is=struct RigVMUnitNode}.Transforms
   ) {is=struct RigVMUnitNode}
   > /Script/ControlRigSpline.RigUnit_DrawControlRigSpline(
    Color=(R=1,G=0,B=0,A=1)
    Detail=16
    Thickness=1
    Spline=@n4.Spline
    Color.R=/Script/ControlRig.RigUnit_GetFloatAnimationChannelFromItem(Item=(Type=Control,Name="ColorR"), bInitial=False) {is=struct RigVMUnitNode, template="GetAnimationChannelFromItem::Execute(out Value,in Item,in bInitial)"}.Value
    Color.G=&RigUnit_GetFloatAnimationChannelFromItem(Item=(Type=Control,Name="ColorG"), bInitial=False) {template="GetAnimationChannelFromItem::Execute(out Value,in Item,in bInitial)"}.Value
    Color.B=&RigUnit_GetFloatAnimationChannelFromItem(Item=(Type=Control,Name="ColorB"), bInitial=False) {template="GetAnimationChannelFromItem::Execute(out Value,in Item,in bInitial)"}.Value
   ) {is=struct RigVMUnitNode}
   GetAnimationChannelFromItem = /Script/ControlRig.RigUnit_GetBoolAnimationChannelFromItem(Item=(Type=Control,Name="Auto Align"), bInitial=False) {is=struct RigVMUnitNode, template="GetAnimationChannelFromItem::Execute(out Value,in Item,in bInitial)"}
   n5 = /Script/ControlRigSpline.RigUnit_TransformFromControlRigSpline(
    UpVector=(X=0,Y=0,Z=1)
    Spline=@n4.Spline
    U=&RigUnit_GetFloatAnimationChannelFromItem(Item=(Type=Control,Name="U Value"), bInitial=False) {template="GetAnimationChannelFromItem::Execute(out Value,in Item,in bInitial)"}.Value
    Roll=&RigUnit_GetFloatAnimationChannelFromItem(Item=(Type=Control,Name="Roll"), bInitial=False) {template="GetAnimationChannelFromItem::Execute(out Value,in Item,in bInitial)"}.Value
   ) {is=struct RigVMUnitNode}
   > &RigUnit_SetTransform(
    Value=(Rotation=(X=0,Y=0,Z=0,W=1),Translation=(X=0,Y=0,Z=0),Scale3D=(X=1,Y=1,Z=1))
    Weight=1
    bInitial=False
    Item=/Script/ControlRig.RigUnit_HierarchyGetParent(bDefaultParent=True, Child=RigVMVariableNode() {variable=Attach Control}.Value) {is=struct RigVMUnitNode}.Parent
    Value.Rotation=RigVMDispatchNode(False=(X=0,Y=0,Z=0,W=1), Condition=@GetAnimationChannelFromItem.Value, True=@n5.Transform.Rotation) {template="DISPATCH_RigVMDispatch_If(in Condition,in True,in False,out Result)"}.Result
    Value.Scale3D=@n5.Transform.Scale3D
    Value.Translation=@n5.Transform.Translation
   )
   Concat = &RigVMFunction_StringConcat(A=chainSim_, B=@To_String.Result) {template="Concat::Execute(in A,in B,out Result)"}
   From_String = RigVMDispatchNode(String=@Concat.Result) {template="DISPATCH_RigDispatch_FromString(in String,out Result)"}
  ]}
]
hierarchy [ComponentSpace {type=Null, loc=<0 0 0>, rot=<0 0 0 1>}]
variables [
 Attach Control {type=FRigElementKey, default=(Type=None,Name="")}
 Bones {type="TArray<FRigElementKey>", default=(())}
 Control {type=FRigElementKey, default=(Type=Bone,Name="")}
 Number of Bones {type=int32, default=6}
 Spline {type=FControlRigSpline, default=()}
]
